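# Niagara skydiving contrail for the BattleSuit cosmetic, with cyan hand glows, sparkles and ribbon trails from hands and feet.
# Spawn groups index bones or sockets on two user skeletal mesh inputs, placing burst-spawned glow sprites with lights, short-lived sparkles thrown against owner velocity, and paired ribbons.
# Time-of-day scalars adjust glow and ribbon alpha.
# A front-end preview flag adds artificial velocity to ribbons and sparkles, simulating motion.
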NiagaraSystem /BRCosmetics/Athena/Effects/Particles/Skydiving/NS_Contrail_BattleSuit/NS_Contrail_BattleSuit
dependencies {
 usedDataInterfaces [&NiagaraDataInterfaceCurve, &NiagaraDataInterfaceSpriteRendererInfo]
 usedDynamicInputs []
 usedModules [&SystemState]
}
emitters {
 Hand_Glow {
  emitterUpdateScript [
   &EmitterState(Life Cycle Mode=::NewEnumerator1, Loop Behavior=::NewEnumerator1, Loop Duration Mode=::NewEnumerator1, UseLoopDelay=false)
   &SpawnBurst_Instantaneous(Spawn Count=4, Use Spawn Probability=false, UseLoopCountLimit=false)
   &SpawnRate(SpawnRate=10, Use Spawn Probability=false) {enabled=false}
  ]
  particleSpawnScript [
   &InitializeParticle(
    Color=<rgba 0.1 1 1 0.5>
    Lifetime=0.5
    Sprite Rotation Angle=&RandomRangeFloat(Minimum=0, Maximum=1)
    Sprite Rotation Mode=::NewEnumerator1
    Sprite Size Mode=::NewEnumerator3
    Uniform Sprite Size=100
    UsePositionOffset=false
    Write Material Random=true
   )
   &FN_SetSpawnGroup()
  ]
  particleUpdateScript [
   &ParticleState(Kill Particles When Lifetime Has Elapsed=false, Let Infinitely Lived Particles Die When Emitter Deactivates=true)
   &ScaleSpriteSize(Uniform Curve Sprite Scale=DataInterface() {propertyValues {Curve {keys [{interpMode=RCIM_Cubic, tangentMode=RCTM_User, time=0, value=0}, {interpMode=RCIM_Cubic, tangentMode=RCTM_User, time=1, value=1}]}}})
   &ScaleColor(
    Scale Alpha=&FloatFromCurve(
     FloatCurve=DataInterface() {propertyValues {Curve {keys [{interpMode=RCIM_Cubic, tangentMode=RCTM_User, time=0, value=0}, {interpMode=RCIM_Cubic, tangentMode=RCTM_User, time=1, value=1}]}}}
     Scale Curve=&FNTimeOfDayScalar(Night Scale=0.3, Morning Scale=0.5, Day Scale=0.75, Evening Scale=0.5)
    )
    ScaleA=true
    ScaleRGB=false
   )
   &SkeletalMeshLocation(Bone / Socket Index=$Particles.SpawnGroup, Bone Sampling Mode=::NewEnumerator4, MaskSpawnGroup=false, Skeletal Mesh=$User.SkeletalMesh)
   &CameraOffset(Camera Offset Amount=16)
   &FN_Light_Attributes(
    Color=<rgba 0.1 1 1 1>
    Create Color Attribute=true
    Create Exponent Attribute=true
    Create Radius Attribute=true
    Exponent=7
    Radius=75
    Scale Color=5
   )
   /SetVariables(Particles.R2_Size:Vector2f=<50 50>) {isSetParameters=true}
  ]
  rendererClasses [&NiagaraSpriteRendererProperties, &NiagaraLightRendererProperties]
  renderers [
   {rendererClass=&NiagaraSpriteRendererProperties}
   {rendererClass=&NiagaraLightRendererProperties, rendererIndex=1}
   {rendererClass=&NiagaraSpriteRendererProperties, rendererIndex=2}
  ]
 }
 Hand_Sparkles {
  emitterUpdateScript [
   &EmitterState()
   &SpawnRate(SpawnRate=60, Use Spawn Probability=false)
  ]
  particleSpawnScript [
   &InitializeParticle(
    Lifetime=&RandomRangeFloat(Minimum=0.1, Maximum=0.2)
    Sprite Size=&RandomRangeVector2D(Minimum=<20 60>, Maximum=<30 80>)
    Sprite Size Mode=::NewEnumerator4
    UsePositionOffset=false
    Write Material Random=true
   )
   &FN_SetSpawnGroup()
   &ShapeLocation(
    Cylinder Height=5
    Cylinder Radius=5
    MaskSpawnGroup=false
    Override Seed=false
    Shape Primitive=::NewEnumerator1
    Surface Only=true
    Use Endcaps In Surface Only Mode=false
    UseOwnerScale=true
   )
   &SkeletalMeshLocation(
    Bone / Socket Index=$Particles.SpawnGroup
    Bone Sampling Mode=::NewEnumerator4
    MaskSpawnGroup=false
    Position Sampling=::NewEnumerator1
    Skeletal Mesh=$User.SkeletalMesh
   )
   &AddVelocity(
    Module.Use Falloff Distance=false
    Override Seed=false
    Velocity Mode=::NewEnumerator1
    Velocity Speed=&RandomRangeFloat(Minimum=100, Maximum=200)
   ) {module=AddVelocity002}
   &AddVelocity(
    Velocity=&Multiply_VectorByFloat(Vector=$Engine.Owner.Velocity, Float=-1)
    Velocity Speed Scale=&RandomRangeFloat(Minimum=0.2, Maximum=0.4)
   )
   &AddVelocity(Velocity=&Multiply_VectorByFloat(
     Vector=<0 -300 250>
     Float=&MakeCustomFloatFromBool(Bool=$User.bIsFrontEndPreview, True Float=1, False Float=0)
    )) {module=AddVelocity001}
  ]
  particleUpdateScript [
   &ParticleState()
   &ScaleSpriteSize(Uniform Curve Sprite Scale=DataInterface() {propertyValues {Curve {keys [{interpMode=RCIM_Cubic, tangentMode=RCTM_User, time=0, value=1}, {interpMode=RCIM_Cubic, tangentMode=RCTM_User, time=1, value=0}]}}})
   &ScaleColor(
    Scale Alpha=&FloatFromCurve(FloatCurve=DataInterface() {propertyValues {Curve {keys [<at 0 0 RCIM_Cubic>,<at 0.5 1 RCIM_Cubic>,<at 1 0 RCIM_Cubic>]}}})
    ScaleA=true
    ScaleRGB=true
   )
   &InheritSourceMovement(Applied Position Scale=<0.7 0.7 0.9>, ApplyForce=false, ApplyPosition=true, ApplyPositionFalloff=false) {enabled=false}
   &SolveForcesAndVelocity(Clamp Velocity=false, Limit Acceleration=false)
   &SubUVAnimation(
    "Sprite Renderer "=DataInterface() {propertyValues {SpriteRenderer=&self:Hand_Sparkles_0.NiagaraSpriteRendererProperties_0}}
    SubUV Animation Mode:ENiagara_SubUVLookupModeV2=::NewEnumerator2
    UseEndFrame=false
    UseStartFrame=false
   )
  ]
  rendererClasses [&NiagaraSpriteRendererProperties]
  renderers [{rendererClass=&NiagaraSpriteRendererProperties}]
 }
 Ribbon_Feet {
  emitterUpdateScript [
   &EmitterState()
   &FN_SpawnRibbon_v2(Number of Ribbons=2, Spawn Mode=::NewEnumerator0, SpawnRate=60)
   &FN_Integrator() {enabled=false}
  ]
  particleSpawnScript [
   /Game/Effects/Niagara/Modules/Ribbons/FN_InitializeRibbon_v4(Enable Skeletal Position=true, Lifetime=0.1, Ribbon Width=30, SkeletalMesh:NiagaraDataInterfaceSkeletalMesh=$User.SkelMesh_Feet)
   &FN_FlipRibbonV() {enabled=false}
   &AddVelocity(Velocity=&Multiply_VectorByFloat(
     Vector=<0 -400 500>
     Float=&MakeCustomFloatFromBool(Bool=$User.bIsFrontEndPreview, True Float=1, False Float=0)
    ))
  ]
  particleUpdateScript [
   &SolveForcesAndVelocity(Clamp Velocity=false, Limit Acceleration=false)
   /Game/Effects/Niagara/Modules/Ribbons/FN_UpdateRibbon_v4(Enable Skeletal Position=true, SkeletalMesh:NiagaraDataInterfaceSkeletalMesh=$User.SkelMesh_Feet)
   &FN_RibbonMaterial() {enabled=false}
   &ScaleColor(
    Scale Alpha=&FloatFromCurve(
     FloatCurve=DataInterface() {propertyValues {Curve {keys [
         {interpMode=RCIM_Cubic, tangentMode=RCTM_User, time=0, value=0}
         {interpMode=RCIM_Cubic, tangentMode=RCTM_User, time=0.3, value=2.5}
         {tangentMode=RCTM_User, time=1, value=0}
        ]}}}
     Scale Curve=&FNTimeOfDayScalar(Night Scale=0.3, Morning Scale=0.5, Evening Scale=0.5)
    )
    ScaleA=true
    ScaleRGB=true
   )
   /SetVariables(Particles.RibbonUVDistance:NiagaraFloat=&Multiply_Float(A=$Particles.NormalizedAge, B=80)) {isSetParameters=true, module=SetVariables_#1}
   &ScaleRibbonWidth(Ribbon Width Scale=&MaskFloatBySpawnGroup(Masked Float=-1, Passthrough Float=1))
   &DynamicMaterialParameters(
    Index 0 Param 1=/Game/Effects/Niagara/DynamicInputs/Random/FN_SeededRandomRangeFloat(Minimum:NiagaraFloat=0, Maximum:NiagaraFloat=1, Random Seed:NiagaraInt32=$Particles.SpawnGroup)
    Module.Param0WriteEnabled=true
    Module.Param1WriteEnabled=true
    Module.Param2WriteEnabled=true
    Module.Param3WriteEnabled=true
   )
  ]
  rendererClasses [&NiagaraRibbonRendererProperties]
  renderers [{rendererClass=&NiagaraRibbonRendererProperties}]
 }
 Ribbon_Hands {
  emitterUpdateScript [
   &EmitterState()
   &FN_SpawnRibbon_v2(Number of Ribbons=2, Spawn Mode=::NewEnumerator0, SpawnRate=60)
   &FN_Integrator() {enabled=false}
  ]
  particleSpawnScript [
   &FN_InitializeRibbon_v4(Enable Skeletal Position=true, Lifetime=0.2, Ribbon Width=40, SkeletalMesh:NiagaraDataInterfaceSkeletalMesh=$User.SkeletalMesh)
   &FN_FlipRibbonV() {enabled=false}
   &AddVelocity(Velocity=&Multiply_VectorByFloat(
     Vector=<0 -400 500>
     Float=&MakeCustomFloatFromBool(Bool=$User.bIsFrontEndPreview, True Float=1, False Float=0)
    ))
  ]
  particleUpdateScript [
   &SolveForcesAndVelocity(Clamp Velocity=false, Limit Acceleration=false)
   &FN_UpdateRibbon_v4(Enable Skeletal Position=true, SkeletalMesh:NiagaraDataInterfaceSkeletalMesh=$User.SkeletalMesh)
   &FN_RibbonMaterial() {enabled=false}
   &ScaleColor(
    Scale Alpha=&FloatFromCurve(
     FloatCurve=DataInterface() {propertyValues {Curve {keys [
         {interpMode=RCIM_Cubic, tangentMode=RCTM_User, time=0, value=0}
         {interpMode=RCIM_Cubic, tangentMode=RCTM_User, time=0.5, value=2.5}
         {tangentMode=RCTM_User, time=1, value=0}
        ]}}}
     Scale Curve=&FNTimeOfDayScalar(Night Scale=0.3, Morning Scale=0.5, Evening Scale=0.5)
    )
    ScaleA=true
    ScaleRGB=true
   )
   /SetVariables(Particles.RibbonUVDistance:NiagaraFloat=&Multiply_Float(A=$Particles.NormalizedAge, B=100)) {isSetParameters=true, module=SetVariables_#1}
   &ScaleRibbonWidth(Ribbon Width Scale=&MaskFloatBySpawnGroup(Masked Float=-1, Passthrough Float=1))
   &DynamicMaterialParameters(
    Index 0 Param 1=&FN_SeededRandomRangeFloat(Minimum:NiagaraFloat=0, Maximum:NiagaraFloat=1, Random Seed:NiagaraInt32=$Particles.SpawnGroup)
    Module.Param0WriteEnabled=true
    Module.Param1WriteEnabled=true
    Module.Param2WriteEnabled=true
    Module.Param3WriteEnabled=true
   )
  ]
  rendererClasses [&NiagaraRibbonRendererProperties]
  renderers [{rendererClass=&NiagaraRibbonRendererProperties}]
 }
}
summary {
 systemName=NS_Contrail_BattleSuit
 emitters [
  {emitterName=Hand_Glow, rendererClasses [&NiagaraSpriteRendererProperties, &NiagaraLightRendererProperties]}
  {emitterName=Ribbon_Hands, rendererClasses [&NiagaraRibbonRendererProperties]}
  {emitterName=Ribbon_Feet, rendererClasses [&NiagaraRibbonRendererProperties]}
  {emitterName=Hand_Sparkles, rendererClasses [&NiagaraSpriteRendererProperties]}
 ]
 userVariables [
  User.bIsFrontEnd {type=NiagaraBool}
  User.bIsFrontEndPreview {type=NiagaraBool}
  User.RibbonIDOffset {type=NiagaraInt32}
  User.SkeletalMesh {
   type=NiagaraDataInterfaceSkeletalMesh
   defaultValue {
    struct=NiagaraExt_VariableValue_DataInterface
    value {dataInterface=&self:NiagaraDataInterfaceSkeletalMesh_0}
   }
  }
  User.SkelMesh_Feet {
   type=NiagaraDataInterfaceSkeletalMesh
   defaultValue {
    struct=NiagaraExt_VariableValue_DataInterface
    value {dataInterface=&self:NiagaraDataInterfaceSkeletalMesh_1}
   }
  }
 ]
}
